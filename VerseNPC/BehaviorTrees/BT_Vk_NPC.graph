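# Root behavior tree for Verse NPCs that switches between the stock Athena NPC behavior and the Verse NPC AI behavior, using the BB_Vk_NPC blackboard.
# A selector first runs the base BT_NPC tree when VK_DefaultBehaviorExecutionStatus is at or above ExecutionPending.
# Otherwise, or if that branch fails, it falls through to the undecorated BT_Vk_NPC_AI subtree.
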BehaviorTree /VerseNPC/BehaviorTrees/BT_Vk_NPC
blackboard=&/BB_Vk_NPC
root {
 name=Selector
 class=BTComposite_Selector
 children [
  Run Default behavior {
   class=BTTask_RunBehavior
   decorators [Default Behavior Status {
     class=BTDecorator_Blackboard
     properties {
      arithmeticOperation=GreaterOrEqual
      intValue=3
      nodeName=Default Behavior Status
      stringValue=EExecutionStatus::ExecutionPending
      blackboardKey {selectedKeyName=VK_DefaultBehaviorExecutionStatus}
     }
    }]
   properties {behaviorAsset=/Game/Athena/AI/NPCs/Base/BehaviorTree/BT_NPC, nodeName=Run Default behavior}
  }
  Run Behavior {class=BTTask_RunBehavior, decorators [], properties {behaviorAsset=&/BT_Vk_NPC_AI, nodeName=Run Behavior}}
 ]
 decorators []
 properties {
  children [
   {
    childComposite=null
    childTask=&self:BTTask_RunBehavior_3
    decorators [&self:BTDecorator_Blackboard_47]
   }
   {childComposite=null, childTask=&self:BTTask_RunBehavior_9, decorators []}
  ]
  services []
 }
 services []
}
